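# Anim Blueprint for the FondBreak backpack cosmetic, adding secondary motion to its head, ears, bows and bow straps.
# The input pose gets a dyn_body offset and is cached, then runs through a chain of spring-constrained Anim Dynamics nodes into a second cache.
# Additive Transform Bone nodes then rotate both strap master bones before output.
# Each update sets GravityOverride for the strap simulations, switching from downward to slightly upward gravity while skydiving.
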AnimBlueprint /BRCosmetics/Accessories/FORT_Backpacks/Backpack_F_MED_FondBreak/Meshes/F_MED_FondBreak_Pack_AnimBP
target_skeleton=&/F_MED_FondBreak_Pack_Skeleton
graphs [
 AnimGraph {nodes [
   n1 = "Transform (Modify) Bone - Bone: dyn_bow_strap_l_mstr"!(ComponentPose=Local To Component(LocalPose=Use cached pose 'CRCahce_dynamics'()), bAlphaBoolEnabled:bool=True, Rotation="3.6,0,0") -> (Pose) {
    alphaInputType=Bool
    rotationMode=BMM_Additive
    translationMode=BMM_Additive
    boneToModify {boneName=dyn_bow_strap_l_mstr}
    rotation {pitch=3.6}
   }
   n2 = "Transform (Modify) Bone - Bone: dyn_bow_strap_r_mstr"!(ComponentPose=@n1, bAlphaBoolEnabled:bool=True, Rotation="-3,0,0") -> (Pose) {
    alphaInputType=Bool
    rotationMode=BMM_Additive
    translationMode=BMM_Additive
    boneToModify {boneName=dyn_bow_strap_r_mstr}
    rotation {pitch=-3}
   }
   n3 = "Transform (Modify) Bone - Bone: dyn_bow_strap_l_mstr"(ComponentPose=@n2, bAlphaBoolEnabled=True, Rotation="3.6,-30,0") {
    alphaInputType=Bool
    rotationMode=BMM_Additive
    translationMode=BMM_Additive
    boneToModify {boneName=dyn_bow_strap_l_mstr}
    rotation {pitch=3.6, yaw=-30}
   }
   n4 = "Transform (Modify) Bone - Bone: dyn_bow_strap_r_mstr"(ComponentPose=@n3, bAlphaBoolEnabled=True, Rotation="-3,30,0") {
    alphaInputType=Bool
    rotationMode=BMM_Additive
    translationMode=BMM_Additive
    boneToModify {boneName=dyn_bow_strap_r_mstr}
    rotation {pitch=-3, yaw=30}
   }
   Output Pose(Result=Component To Local(ComponentPose=@n4)) {layerGroup=None}
   n5 = Local To Component(LocalPose=Input Pose())
   Save cached pose 'CRCache_inputPose'(Pose=Component To Local(ComponentPose="Transform (Modify) Bone - Bone: dyn_body"!(ComponentPose=@n5, Alpha=1, Translation="0,1.7,0") -> (Pose) {translationMode=BMM_Additive, boneToModify {boneName=dyn_body}, translation {y=1.7}})) {cacheName=CRCache_inputPose}
   n6 = Anim Dynamics - Bone: dyn_head!(ComponentPose=Local To Component(LocalPose=Use cached pose 'CRCache_inputPose'()), Alpha=1) -> (Pose) {
    angularDampingOverride=0.9998
    angularSpringConstant=5000
    bAngularSpring=true
    bOverrideAngularDamping=true
    bUseGravityOverride=true
    simulationSpace=RootRelative
    alphaBoolBlend {blendInTime=0.1, blendOption=Linear, blendOutTime=0.1}
    boundBone {boneName=dyn_head}
    componentAppliedLinearAccClamp=<700 700 1000>
    componentLinearAccScale=<0.6 0.6 0.6>
    componentLinearVelScale=<0.8 0.8 0.8>
    gravityOverride {z=-500}
    physicsBodyDefinitions [{
      boundBone {boneName=dyn_head}
      boxExtents=<25 25 25>
      constraintSetup {angularLimitsMax=<10 35 0>, angularLimitsMin=<-10 -35 -25>, angularTarget=<1 0 0>}
      localJointOffset=<2 0 0>
     }]
   }
   n7 = Anim Dynamics - Bone: dyn_ear_r_01!(ComponentPose=@n6, Alpha=1) -> (Pose) {
    angularDampingOverride=0.9998
    angularSpringConstant=5000
    bAngularSpring=true
    bOverrideAngularDamping=true
    bUseGravityOverride=true
    simulationSpace=RootRelative
    alphaBoolBlend {blendInTime=0.1, blendOption=Linear, blendOutTime=0.1}
    boundBone {boneName=dyn_ear_r_01}
    componentAppliedLinearAccClamp=<500 500 1000>
    componentLinearAccScale=<0.6 0.6 0.6>
    componentLinearVelScale=<0.8 0.8 0.8>
    gravityOverride {z=-500}
    physicsBodyDefinitions [{
      boundBone {boneName=dyn_ear_r_01}
      boxExtents=<20 20 20>
      constraintSetup {angularLimitsMax=<2.5 50 15>, angularLimitsMin=<-2.5 -30 -25>, angularTarget=<1 0 0>}
      localJointOffset=<2 0 0>
     }]
   }
   n8 = Anim Dynamics - Bone: dyn_ear_r_02!(ComponentPose=@n7, Alpha=1) -> (Pose) {
    angularDampingOverride=0.9999
    angularSpringConstant=4500
    bAngularSpring=true
    bOverrideAngularDamping=true
    bUseGravityOverride=true
    simulationSpace=RootRelative
    alphaBoolBlend {blendInTime=0.1, blendOption=Linear, blendOutTime=0.1}
    boundBone {boneName=dyn_ear_r_02}
    componentAppliedLinearAccClamp=<500 500 1000>
    componentLinearAccScale=<0.6 0.6 0.6>
    componentLinearVelScale=<0.8 0.8 0.8>
    gravityOverride {z=-500}
    physicsBodyDefinitions [{
      boundBone {boneName=dyn_ear_r_02}
      boxExtents=<15 15 15>
      constraintSetup {angularLimitsMax=<2.5 50 15>, angularLimitsMin=<-2.5 -30 -25>, angularTarget=<1 0 0>}
      localJointOffset=<1 0 0>
     }]
   }
   n9 = Anim Dynamics - Bone: dyn_ear_l_01!(ComponentPose=@n8, Alpha=1) -> (Pose) {
    angularDampingOverride=0.9998
    angularSpringConstant=5000
    bAngularSpring=true
    bOverrideAngularDamping=true
    bUseGravityOverride=true
    simulationSpace=RootRelative
    alphaBoolBlend {blendInTime=0.1, blendOption=Linear, blendOutTime=0.1}
    boundBone {boneName=dyn_ear_l_01}
    componentAppliedLinearAccClamp=<500 500 1000>
    componentLinearAccScale=<0.6 0.6 0.6>
    componentLinearVelScale=<0.8 0.8 0.8>
    gravityOverride {z=-500}
    physicsBodyDefinitions [{
      boundBone {boneName=dyn_ear_l_01}
      boxExtents=<20 20 20>
      constraintSetup {angularLimitsMax=<2.5 30 15>, angularLimitsMin=<-2.5 -50 -25>, angularTarget=<1 0 0>}
      localJointOffset=<2 0 0>
     }]
   }
   n10 = Anim Dynamics - Bone: dyn_ear_l_02!(ComponentPose=@n9, Alpha=1) -> (Pose) {
    angularDampingOverride=0.9999
    angularSpringConstant=4500
    bAngularSpring=true
    bOverrideAngularDamping=true
    bUseGravityOverride=true
    simulationSpace=RootRelative
    alphaBoolBlend {blendInTime=0.1, blendOption=Linear, blendOutTime=0.1}
    boundBone {boneName=dyn_ear_l_02}
    componentAppliedLinearAccClamp=<500 500 1000>
    componentLinearAccScale=<0.6 0.6 0.6>
    componentLinearVelScale=<0.8 0.8 0.8>
    gravityOverride {z=-500}
    physicsBodyDefinitions [{
      boundBone {boneName=dyn_ear_l_02}
      boxExtents=<15 15 15>
      constraintSetup {angularLimitsMax=<2.5 30 15>, angularLimitsMin=<-2.5 -50 -25>, angularTarget=<1 0 0>}
      localJointOffset=<1 0 0>
     }]
   }
   n11 = Anim Dynamics - Bone: dyn_ear_l_03!(ComponentPose=@n10, Alpha=1) -> (Pose) {
    angularDampingOverride=0.9999
    angularSpringConstant=2500
    bAngularSpring=true
    bOverrideAngularDamping=true
    bUseGravityOverride=true
    simulationSpace=RootRelative
    alphaBoolBlend {blendInTime=0.1, blendOption=Linear, blendOutTime=0.1}
    boundBone {boneName=dyn_ear_l_03}
    componentAppliedLinearAccClamp=<500 500 1000>
    componentLinearAccScale=<0.6 0.6 0.6>
    componentLinearVelScale=<0.8 0.8 0.8>
    gravityOverride {z=-500}
    physicsBodyDefinitions [{
      boundBone {boneName=dyn_ear_l_03}
      boxExtents=<10 10 10>
      constraintSetup {angularLimitsMax=<2.5 30 10>, angularLimitsMin=<-2.5 -30 -10>, angularTarget=<1 0 0>}
      localJointOffset=<0.5 0 0>
     }]
   }
   n12 = Anim Dynamics - Bone: dyn_bow_r_a!(ComponentPose=@n11, Alpha=1) -> (Pose) {
    angularDampingOverride=0.99
    angularSpringConstant=1000
    bAngularSpring=true
    bOverrideAngularDamping=true
    bUseGravityOverride=true
    boundBone {boneName=dyn_bow_r_a}
    componentAppliedLinearAccClamp=<1000 1000 1000>
    componentLinearAccScale=<0.4 0.4 0.4>
    componentLinearVelScale=<0.4 0.4 0.4>
    gravityOverride {z=-980}
    physicsBodyDefinitions [{
      boundBone {boneName=dyn_bow_r_a}
      boxExtents=<10 10 10>
      constraintSetup {angularLimitsMax=<0 0 0>, angularLimitsMin=<0 -25 -50>, angularTarget=<1 0 0>}
      localJointOffset=<1 0 0>
     }]
   }
   n13 = Anim Dynamics - Bone: dyn_bow_r_b!(ComponentPose=@n12, Alpha=1) -> (Pose) {
    angularDampingOverride=0.99
    angularSpringConstant=1000
    bAngularSpring=true
    bOverrideAngularDamping=true
    bUseGravityOverride=true
    boundBone {boneName=dyn_bow_r_b}
    componentAppliedLinearAccClamp=<1000 1000 1000>
    componentLinearAccScale=<0.4 0.4 0.4>
    componentLinearVelScale=<0.4 0.4 0.4>
    gravityOverride {z=-980}
    physicsBodyDefinitions [{
      boundBone {boneName=dyn_bow_r_b}
      boxExtents=<10 10 10>
      constraintSetup {angularLimitsMax=<0 0 0>, angularLimitsMin=<0 0 -50>, angularTarget=<1 0 0>}
      localJointOffset=<1 0 0>
     }]
   }
   n14 = Anim Dynamics - Bone: dyn_bow_strap_r_01!(ComponentPose=@n13, Alpha=1, GravityOverride:Vector="0,0,-980") -> (Pose) {
    angularDampingOverride=0.99
    angularSpringConstant=1000
    bAngularSpring=true
    bOverrideAngularDamping=true
    bUseGravityOverride=true
    boundBone {boneName=dyn_bow_strap_r_01}
    componentAppliedLinearAccClamp=<1500 1500 1500>
    componentLinearAccScale=<0.4 0.4 0.4>
    componentLinearVelScale=<0.4 0.4 0.4>
    gravityOverride {z=-980}
    physicsBodyDefinitions [{
      boundBone {boneName=dyn_bow_strap_r_01}
      boxExtents=<10 10 10>
      constraintSetup {angularLimitsMax=<0 5 0>, angularLimitsMin=<0 -15 -35>, angularTarget=<1 0 0>}
      localJointOffset=<1 0 0>
     }]
   }
   n15 = Anim Dynamics - Bone: dyn_bow_strap_r_02!(ComponentPose=@n14, Alpha=1, GravityOverride:Vector="0,0,-980") -> (Pose) {
    angularDampingOverride=0.99
    angularSpringConstant=1000
    bAngularSpring=true
    bOverrideAngularDamping=true
    bUseGravityOverride=true
    boundBone {boneName=dyn_bow_strap_r_02}
    componentAppliedLinearAccClamp=<1500 1500 1500>
    componentLinearAccScale=<0.4 0.4 0.4>
    componentLinearVelScale=<0.4 0.4 0.4>
    gravityOverride {z=-980}
    physicsBodyDefinitions [{
      boundBone {boneName=dyn_bow_strap_r_02}
      boxExtents=<10 10 10>
      constraintSetup {angularLimitsMax=<0 5 0>, angularLimitsMin=<0 -20 -35>, angularTarget=<1 0 0>}
      localJointOffset=<1 0 0>
     }]
   }
   n16 = Anim Dynamics - Bone: dyn_bow_strap_r_03!(ComponentPose=@n15, Alpha=1, GravityOverride:Vector="0,0,-980") -> (Pose) {
    angularDampingOverride=0.99
    angularSpringConstant=1000
    bAngularSpring=true
    bOverrideAngularDamping=true
    bUseGravityOverride=true
    boundBone {boneName=dyn_bow_strap_r_03}
    componentAppliedLinearAccClamp=<1500 1500 1500>
    componentLinearAccScale=<0.4 0.4 0.4>
    componentLinearVelScale=<0.4 0.4 0.4>
    gravityOverride {z=-980}
    physicsBodyDefinitions [{
      boundBone {boneName=dyn_bow_strap_r_03}
      boxExtents=<10 10 10>
      constraintSetup {angularLimitsMax=<0 5 0>, angularLimitsMin=<0 -25 -35>, angularTarget=<1 0 0>}
      localJointOffset=<1 0 0>
     }]
   }
   n17 = Anim Dynamics - Bone: dyn_bow_l_a!(ComponentPose=@n16, Alpha=1) -> (Pose) {
    angularDampingOverride=0.99
    angularSpringConstant=1000
    bAngularSpring=true
    bOverrideAngularDamping=true
    bUseGravityOverride=true
    boundBone {boneName=dyn_bow_l_a}
    componentAppliedLinearAccClamp=<1000 1000 1000>
    componentLinearAccScale=<0.4 0.4 0.4>
    componentLinearVelScale=<0.4 0.4 0.4>
    gravityOverride {z=-980}
    physicsBodyDefinitions [{
      boundBone {boneName=dyn_bow_l_a}
      boxExtents=<10 10 10>
      constraintSetup {angularLimitsMax=<0 25 0>, angularLimitsMin=<0 0 -50>, angularTarget=<1 0 0>}
      localJointOffset=<1 0 0>
     }]
   }
   n18 = Anim Dynamics - Bone: dyn_bow_l_b!(ComponentPose=@n17, Alpha=1) -> (Pose) {
    angularDampingOverride=0.99
    angularSpringConstant=1000
    bAngularSpring=true
    bOverrideAngularDamping=true
    bUseGravityOverride=true
    boundBone {boneName=dyn_bow_l_b}
    componentAppliedLinearAccClamp=<1000 1000 1000>
    componentLinearAccScale=<0.4 0.4 0.4>
    componentLinearVelScale=<0.4 0.4 0.4>
    gravityOverride {z=-980}
    physicsBodyDefinitions [{
      boundBone {boneName=dyn_bow_l_b}
      boxExtents=<10 10 10>
      constraintSetup {angularLimitsMax=<0 0 50>, angularLimitsMin=<0 0 0>, angularTarget=<1 0 0>}
      localJointOffset=<1 0 0>
     }]
   }
   n19 = Anim Dynamics - Bone: dyn_bow_strap_l_01!(ComponentPose=@n18, Alpha=1, GravityOverride:Vector="0,0,-980") -> (Pose) {
    angularDampingOverride=0.99
    angularSpringConstant=1000
    bAngularSpring=true
    bOverrideAngularDamping=true
    bUseGravityOverride=true
    boundBone {boneName=dyn_bow_strap_l_01}
    componentAppliedLinearAccClamp=<1500 1500 1500>
    componentLinearAccScale=<0.5 0.5 0.5>
    componentLinearVelScale=<0.5 0.5 0.5>
    gravityOverride {z=-980}
    physicsBodyDefinitions [{
      boundBone {boneName=dyn_bow_strap_l_01}
      boxExtents=<10 10 10>
      constraintSetup {angularLimitsMax=<0 5 35>, angularLimitsMin=<0 -15 0>, angularTarget=<1 0 0>}
      localJointOffset=<1 0 0>
     }]
   }
   n20 = Anim Dynamics - Bone: dyn_bow_strap_l_02!(ComponentPose=@n19, Alpha=1, GravityOverride:Vector="0,0,-980") -> (Pose) {
    angularDampingOverride=0.99
    angularSpringConstant=1000
    bAngularSpring=true
    bOverrideAngularDamping=true
    bUseGravityOverride=true
    boundBone {boneName=dyn_bow_strap_l_02}
    componentAppliedLinearAccClamp=<1500 1500 1500>
    componentLinearAccScale=<0.5 0.5 0.5>
    componentLinearVelScale=<0.5 0.5 0.5>
    gravityOverride {z=-980}
    physicsBodyDefinitions [{
      boundBone {boneName=dyn_bow_strap_l_02}
      boxExtents=<10 10 10>
      constraintSetup {angularLimitsMax=<0 5 35>, angularLimitsMin=<0 -20 0>, angularTarget=<1 0 0>}
      localJointOffset=<1 0 0>
     }]
   }
   n21 = Anim Dynamics - Bone: dyn_bow_strap_l_03!(ComponentPose=@n20, Alpha=1, GravityOverride:Vector="0,0,-980") -> (Pose) {
    angularDampingOverride=0.99
    angularSpringConstant=1000
    bAngularSpring=true
    bOverrideAngularDamping=true
    bUseGravityOverride=true
    boundBone {boneName=dyn_bow_strap_l_03}
    componentAppliedLinearAccClamp=<1500 1500 1500>
    componentLinearAccScale=<0.5 0.5 0.5>
    componentLinearVelScale=<0.5 0.5 0.5>
    gravityOverride {z=-980}
    physicsBodyDefinitions [{
      boundBone {boneName=dyn_bow_strap_l_03}
      boxExtents=<10 10 10>
      constraintSetup {angularLimitsMax=<0 5 35>, angularLimitsMin=<0 -25 0>, angularTarget=<1 0 0>}
      localJointOffset=<1 0 0>
     }]
   }
   Save cached pose 'CRCahce_dynamics'(Pose=Component To Local(ComponentPose=@n21)) {cacheName=CRCahce_dynamics}
  ]}
 EventGraph {nodes [
   Event BlueprintUpdateAnimation()
   > Set GravityOverride(GravityOverride=SelectVector(A="0,0,50", B="0,0,-980", bPickA=Get bIsSkyDiving()))
   TryGetPawnOwner()
  ]}
]
variables [GravityOverride {type=Vector}]
